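# Translucent UI material that draws clouds for the WildEstate intro, with a choice of three cloud shapes.
# The R, G and B channels of one texture act as separate distance fields, each thresholded through an anti-aliasing function.
# The SwitchCloud scalar parameter blends between the three resulting shapes through Lerp_Multiple_Float3.
# A rotated linear gradient drives a three-color blend from white through pale gray-blue to darker blue.
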Material /WildEstateUI/Introduction/Material/M_UI_WildEstateIntro_Clouds
graph {
 properties {bAutomaticallySetUsageInEditor=false, blendMode=BLEND_Translucent, materialDomain=MD_UI, usedShadingModels=Default Lit}
 nodes [
  n1 = TextureSample() {texture=/WildEstateUI/Introduction/Textures/T_UI_Cinematic_Intro_Clouds}
  n2 = Constant() {r=0.5}
  n3 = Constant() {r=0.5}
  n4 = Constant() {r=0.5}
  MaterialFunctionCall(
   1=NamedRerouteDeclaration(None=MaterialFunctionCall(DistanceField=@n1.R, Threshold=@n2) {materialFunction=&MF_UI_AntiAliasing}.Outer) {name=Cloud1, nodeColor=<rgba 0 0 0 1>}
   2=NamedRerouteDeclaration(None=MaterialFunctionCall(DistanceField=@n1.G, Threshold=@n3) {materialFunction=&MF_UI_AntiAliasing}.Outer) {name=Cloud2, nodeColor=<rgba 0 0 0 1>}
   3=NamedRerouteDeclaration(None=MaterialFunctionCall(DistanceField=@n1.B, Threshold=@n4) {materialFunction=&MF_UI_AntiAliasing}.Outer) {name=Cloud3, nodeColor=<rgba 0 0.6484 1 1>}
   A=ScalarParameter() {parameterName=SwitchCloud}
  ) {materialFunction=&Lerp_Multiple_Float3}
  MaterialFunctionCall(
   A=Constant3Vector() {constant=<rgba 1 1 1 0>}
   B=Constant3Vector() {constant=<rgba 0.5647 0.6514 0.6939 1>}
   C=Constant3Vector() {constant=<rgba 0.166 0.2709 0.3698 1>}
   Alpha=MaterialFunctionCall(GradientEnd=Constant() {r=0.616}, RotationAngle=Constant() {r=-0.25}) {materialFunction=&MF_UI_LinearGradient}.Result
  ) {materialFunction=&3ColorBlend}
 ]
}
