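# UI-domain material that draws a two-tone blue gradient background, named for the WildEstate introduction screen.
# A RadialGradientExponential function, centered on the bottom middle of the UV space, drives a lerp between the MainBlue and LightBlue vector parameters.
# The result is lightest at the bottom center and fades outward to the deeper blue, with no textures or animation.
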Material /WildEstateUI/Introduction/Material/M_UI_WildEstateIntro_Background
graph {
 properties {bAutomaticallySetUsageInEditor=false, materialDomain=MD_UI, usedShadingModels=Default Lit}
 nodes [LinearInterpolate(
   A=VectorParameter() {parameterName=MainBlue, defaultValue=<rgba 0.003 0.1413 0.491 1>}.RGB
   B=VectorParameter() {parameterName=LightBlue, defaultValue=<rgba 0.0612 0.3325 0.5972 1>}.RGB
   Alpha=MaterialFunctionCall(CenterPosition=Constant2Vector() {g=1, r=0.5}, Radius=Constant() {r=1}) {materialFunction=&RadialGradientExponential}.RadialGradientExponential
  )]
}
