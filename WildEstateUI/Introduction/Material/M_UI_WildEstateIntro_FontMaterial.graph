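# Translucent UI material, named as the font material for the WildEstate introduction, with gold FontColor and darker FontOutline parameters.
# A diagonally panning noise texture slightly offsets the texture coordinates, which are split, recombined and fed into an MF_UI_Box function as wobbling UVs.
# A time-driven sine, both color parameters and an empty lerp sit unconnected, and the graph shows no link to a material output.
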Material /WildEstateUI/Introduction/Material/M_UI_WildEstateIntro_FontMaterial
graph {
 properties {bAutomaticallySetUsageInEditor=false, blendMode=BLEND_Translucent, materialDomain=MD_UI, usedShadingModels=Default Lit}
 nodes [
  LinearInterpolate()
  n1 = Panner(Coordinate=TextureCoordinate()) {speedX=-0.1, speedY=0.1}
  n2 = Add(
   A=TextureCoordinate()
   B=Multiply(A=TextureSample(UVs=@n1) {texture=/WildEstateUI/Introduction/Textures/T_UI_WildEstateNoiseExample}.R) {constB=0.01}
  )
  n3 = ComponentMask(None=@n2) {r=true}
  MaterialFunctionCall(UVs=AppendVector(A=Add(A=@n3) {constB=0}, B=ComponentMask(None=@n2) {g=true}), Size=Constant2Vector() {g=0.5, r=0.9}) {materialFunction=&MF_UI_Box}
  n4 = Time()
  Sine(None=Frac(None=Multiply(A=@n4) {constB=0.1}))
  VectorParameter() {parameterName=FontColor, defaultValue=<rgba 0.7379 0.3763 0 1>}
  VectorParameter() {parameterName=FontOutline, defaultValue=<rgba 0.1406 0.0717 0 1>}
 ]
}
