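# Intro title widget for WildEstateUI that shows "Battle Royale" over a material-driven background and four cloud images.
# TextAnimation slides the clouds outward and fades them while the title, styled with custom fill and outline font materials, scales up and then fades out.
# The event graph holds only empty stub events, so the widget does not start the animation itself.
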WidgetBlueprint /WildEstateUI/Introduction/WBP_WildEstate_Intro
animations [TextAnimation {
  binding_count=5
  length_seconds=2
  object=&self:TextAnimation
  sequence {bindings [
    Image_Cloud2 {
     id=#1
     possessed_class=Image
     tracks [
      {
       type=MovieScene2DTransformTrack
       sections [{
         type=MovieScene2DTransformSection
         end_frame=15
         start_frame=0
         channels [
          Translation.X {type=MovieSceneScriptingFloatChannel, keys [[0,0],[15,-126.752]]}
          Translation.Y {type=MovieSceneScriptingFloatChannel, keys [[0,512],[15,709.9318]]}
         ]
        }]
      }
      {
       type=MovieSceneFloatTrack
       display_name=Render Opacity
       sections [{type=MovieSceneFloatSection, end_frame=30, start_frame=0, channels [None {type=MovieSceneScriptingFloatChannel, keys [[0,1],[30,0]]}]}]
      }
     ]
    }
    Image_Cloud1 {
     id=#2
     possessed_class=Image
     tracks [
      {
       type=MovieScene2DTransformTrack
       sections [{
         type=MovieScene2DTransformSection
         end_frame=15
         start_frame=0
         channels [
          Translation.X {type=MovieSceneScriptingFloatChannel, keys [[0,-128],[15,-811.3011]]}
          Translation.Y {type=MovieSceneScriptingFloatChannel, const [-128,1]}
         ]
        }]
      }
      {
       type=MovieSceneFloatTrack
       display_name=Render Opacity
       sections [{type=MovieSceneFloatSection, end_frame=30, start_frame=0, channels [None {type=MovieSceneScriptingFloatChannel, keys [[0,1],[30,0]]}]}]
      }
     ]
    }
    Image_Cloud3 {
     id=#3
     possessed_class=Image
     tracks [
      {
       type=MovieScene2DTransformTrack
       sections [{
         type=MovieScene2DTransformSection
         end_frame=20
         start_frame=0
         channels [
          Translation.X {type=MovieSceneScriptingFloatChannel, keys [[0,1000],[20,1600]]}
          Translation.Y {type=MovieSceneScriptingFloatChannel, const [0,1]}
         ]
        }]
      }
      {
       type=MovieSceneFloatTrack
       display_name=Render Opacity
       sections [{type=MovieSceneFloatSection, end_frame=30, start_frame=0, channels [None {type=MovieSceneScriptingFloatChannel, keys [[0,1],[30,0]]}]}]
      }
     ]
    }
    Image_Cloud4 {
     id=#4
     possessed_class=Image
     tracks [
      {
       type=MovieScene2DTransformTrack
       sections [{
         type=MovieScene2DTransformSection
         end_frame=22
         start_frame=0
         channels [
          Translation.X {type=MovieSceneScriptingFloatChannel, keys [[0,1000],[22,1700]]}
          Translation.Y {type=MovieSceneScriptingFloatChannel, const [-200,1]}
         ]
        }]
      }
      {
       type=MovieSceneFloatTrack
       display_name=Render Opacity
       sections [{type=MovieSceneFloatSection, end_frame=30, start_frame=0, channels [None {type=MovieSceneScriptingFloatChannel, keys [[0,1],[30,0]]}]}]
      }
     ]
    }
    CommonTextBlock_37 {
     id=#5
     possessed_class=CommonTextBlock
     tracks [
      {
       type=MovieScene2DTransformTrack
       sections [{
         type=MovieScene2DTransformSection
         end_frame=40
         start_frame=0
         channels [
          Scale.X {type=MovieSceneScriptingFloatChannel, keys [[0,0.2],[40,4]]}
          Scale.Y {type=MovieSceneScriptingFloatChannel, keys [[0,0.2],[40,4]]}
         ]
        }]
      }
      {
       type=MovieSceneFloatTrack
       display_name=Render Opacity
       sections [{type=MovieSceneFloatSection, end_frame=40, start_frame=0, channels [None {type=MovieSceneScriptingFloatChannel, keys [[0,1],[30,1],[40,0]]}]}]
      }
     ]
    }
   ]}
 }]
graphs [EventGraph {nodes [
   Event PreConstruct()
   Event Construct()
   Event Tick()
  ]}]
variables []
widgets {
 info {inheritedWidgetCount=0, parentClass=&UserWidget, rootWidgetClass=&Overlay, widgetCount=8}
 widgets [
  Overlay_27 {widgetClassPath=&Overlay, widget=Overlay_27}
  Image_Backround {
   widgetClassPath=&Image
   bIsVariable=true
   parent=Overlay_27
   widget=Image_Backround
   properties {brush {
     bIsDynamicallyLoaded=false
     drawAs=Image
     imageType=FullColor
     mirroring=NoMirror
     resourceName=None
     resourceObject=/WildEstateUI/Introduction/Material/M_UI_WildEstateIntro_Background
     tiling=NoTile
     imageSize=<32 32>
     margin=<ltrb 0 0 0 0>
     outlineSettings {
      bUseBrushTransparency=false
      roundingType=HalfHeightRadius
      width=0
      color {colorUseRule=UseColor_Specified, specifiedColor=<rgba 0 0 0 0>}
      cornerRadii=<0 0 0 0>
     }
     tintColor {colorUseRule=UseColor_Specified, specifiedColor=<rgba 1 1 1 1>}
     uVRegion {bIsValid=false, max=<0 0>, min=<0 0>}
    }}
   slot {class=OverlaySlot, properties {horizontalAlignment=HAlign_Fill, verticalAlignment=VAlign_Fill}}
  }
  Clouds {widgetClassPath=&Overlay, parent=Overlay_27, widget=Clouds, slot {class=OverlaySlot}}
  Image_Cloud4 {
   widgetClassPath=&Image
   bIsVariable=true
   parent=Clouds
   widget=Image_Cloud4
   properties {
    brush {
     bIsDynamicallyLoaded=false
     drawAs=Image
     imageType=FullColor
     mirroring=NoMirror
     resourceName=None
     resourceObject=/WildEstateUI/Introduction/Material/MI_UI_WildEstateIntro_Clouds_1
     tiling=NoTile
     imageSize=<1400 1200>
     margin=<ltrb 0 0 0 0>
     outlineSettings {
      bUseBrushTransparency=false
      roundingType=HalfHeightRadius
      width=0
      color {colorUseRule=UseColor_Specified, specifiedColor=<rgba 0 0 0 0>}
      cornerRadii=<0 0 0 0>
     }
     tintColor {colorUseRule=UseColor_Specified, specifiedColor=<rgba 1 1 1 1>}
     uVRegion {bIsValid=false, max=<0 0>, min=<0 0>}
    }
    renderTransform {angle=0, scale=<1 1>, shear=<0 0>, translation=<1000 -200>}
   }
   slot {class=OverlaySlot}
  }
  Image_Cloud3 {
   widgetClassPath=&Image
   bIsVariable=true
   parent=Clouds
   widget=Image_Cloud3
   properties {
    brush {
     bIsDynamicallyLoaded=false
     drawAs=Image
     imageType=FullColor
     mirroring=NoMirror
     resourceName=None
     resourceObject=/WildEstateUI/Introduction/Material/MI_UI_WildEstateIntro_Clouds_3
     tiling=NoTile
     imageSize=<1400 1200>
     margin=<ltrb 0 0 0 0>
     outlineSettings {
      bUseBrushTransparency=false
      roundingType=HalfHeightRadius
      width=0
      color {colorUseRule=UseColor_Specified, specifiedColor=<rgba 0 0 0 0>}
      cornerRadii=<0 0 0 0>
     }
     tintColor {colorUseRule=UseColor_Specified, specifiedColor=<rgba 1 1 1 1>}
     uVRegion {bIsValid=false, max=<0 0>, min=<0 0>}
    }
    renderTransform {angle=0, scale=<1 1>, shear=<0 0>, translation=<1000 0>}
   }
   slot {class=OverlaySlot}
  }
  Image_Cloud2 {
   widgetClassPath=&Image
   bIsVariable=true
   parent=Clouds
   widget=Image_Cloud2
   properties {
    brush {
     bIsDynamicallyLoaded=false
     drawAs=Image
     imageType=FullColor
     mirroring=NoMirror
     resourceName=None
     resourceObject=/WildEstateUI/Introduction/Material/MI_UI_WildEstateIntro_Clouds_2
     tiling=NoTile
     imageSize=<1200 1200>
     margin=<ltrb 0 0 0 0>
     outlineSettings {
      bUseBrushTransparency=false
      roundingType=HalfHeightRadius
      width=0
      color {colorUseRule=UseColor_Specified, specifiedColor=<rgba 0 0 0 0>}
      cornerRadii=<0 0 0 0>
     }
     tintColor {colorUseRule=UseColor_Specified, specifiedColor=<rgba 1 1 1 1>}
     uVRegion {bIsValid=false, max=<0 0>, min=<0 0>}
    }
    renderTransform {angle=0, scale=<1 1>, shear=<0 0>, translation=<0 512>}
   }
   slot {class=OverlaySlot}
  }
  Image_Cloud1 {
   widgetClassPath=&Image
   bIsVariable=true
   parent=Clouds
   widget=Image_Cloud1
   properties {
    brush {
     bIsDynamicallyLoaded=false
     drawAs=Image
     imageType=FullColor
     mirroring=NoMirror
     resourceName=None
     resourceObject=&MI_UI_WildEstateIntro_Clouds_1
     tiling=NoTile
     imageSize=<1200 1200>
     margin=<ltrb 0 0 0 0>
     outlineSettings {
      bUseBrushTransparency=false
      roundingType=HalfHeightRadius
      width=0
      color {colorUseRule=UseColor_Specified, specifiedColor=<rgba 0 0 0 0>}
      cornerRadii=<0 0 0 0>
     }
     tintColor {colorUseRule=UseColor_Specified, specifiedColor=<rgba 1 1 1 1>}
     uVRegion {bIsValid=false, max=<0 0>, min=<0 0>}
    }
    renderTransform {angle=0, scale=<1 1>, shear=<0 0>, translation=<-128 -128>}
   }
   slot {class=OverlaySlot}
  }
  CommonTextBlock_37 {
   widgetClassPath=&CommonTextBlock
   parent=Overlay_27
   widget=CommonTextBlock_37
   properties {
    bIsScrollingEnabled=false
    justification=Center
    text="Battle\r\nRoyale"
    font {
     bForceMonospaced=false
     bMaterialIsStencil=false
     fontMaterial=/WildEstateUI/Introduction/Material/M_UI_WildEstateIntro_FontMaterial
     fontObject=&Burbank
     letterSpacing=0
     monospacedWidth=1
     size=162
     skewAmount=0
     typefaceFontName=Burbank-Black
     outlineSettings {
      bApplyOutlineToDropShadows=false
      bMiteredCorners=false
      bSeparateFillAlpha=false
      outlineMaterial=/WildEstateUI/Introduction/Material/MI_UI_WildEstateIntro_FontMaterial_Outline
      outlineSize=4
      outlineColor=<rgba 1 1 1 1>
     }
    }
    shadowOffset=<0 0>
   }
   slot {class=OverlaySlot, properties {horizontalAlignment=HAlign_Center, verticalAlignment=VAlign_Center}}
  }
 ]
}
